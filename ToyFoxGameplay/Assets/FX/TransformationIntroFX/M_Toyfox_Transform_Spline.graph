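# Additive unlit, two-sided material for Niagara mesh spline streaks in the Toyfox transformation intro FX.
# SplineThicken, with width multiplied by object scale, combines with a camera offset to displace vertices.
# A soft SphereMask over stretched UVs slides along V via particle DynamicParameter Param1, creating a moving highlight.
# TwoSidedSign lerps that opacity-scaled mask toward zero on front faces, while a bright tint passes through MF_FX_EmissiveAdjustments.
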Material /ToyFoxGameplay/Assets/FX/TransformationIntroFX/M_Toyfox_Transform_Spline
graph {
 properties {
  bAutomaticallySetUsageInEditor=false
  bUsedWithNiagaraMeshParticles=true
  blendMode=BLEND_Additive
  shadingModel=MSM_Unlit
  twoSided=true
  usedShadingModels=Unlit
 }
 nodes [
  n1 = Multiply(
   A=ScalarParameter() {defaultValue=1.5, parameterName=Width}
   B=MaterialFunctionCall() {materialFunction=&ObjectScale}.Scale XYZ
  )
  Add(
   A=MaterialFunctionCall(WidthBase=@n1, WidthTip=@n1) {materialFunction=&SplineThicken}.WorldPositionOffset
   B=MaterialFunctionCall(Offset Amount=ScalarParameter() {defaultValue=250}) {materialFunction=/Engine/Functions/Engine_MaterialFunctions02/WorldPositionOffset/CameraOffset}.Camera Offset
  )
  n2 = Constant() {r=1.7}
  n3 = Multiply(A=TextureCoordinate(), B=AppendVector(A=Constant() {r=1}, B=@n2))
  n4 = ScalarParameter() {defaultValue=-0.5, parameterName=Param_3}
  n5 = ScalarParameter() {defaultValue=1.5, parameterName=Param_4}
  n6 = DynamicParameter() {defaultValue=<rgba 0.5 1 1 1>}
  n7 = AppendVector(
   A=ScalarParameter() {defaultValue=0.5, parameterName=Param_1}
   B=Multiply(A=LinearInterpolate(A=@n4, B=@n5, Alpha=@n6.Param1), B=@n2) {constB=2.3}
  )
  LinearInterpolate(
   A=Multiply(A=SphereMask(A=@n3, B=@n7) {attenuationRadius=0.5, hardnessPercent=0}, B=ScalarParameter() {defaultValue=1, parameterName=Opacity})
   Alpha=TwoSidedSign()
  ) {constB=0}
  MaterialFunctionCall(Emissive=Multiply(A=VectorParameter() {defaultValue=<rgba 0.0023 0.2001 1 1>}.RGB, B=ScalarParameter() {defaultValue=2500, parameterName=Emissive})) {materialFunction=&MF_FX_EmissiveAdjustments}
 ]
}
