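# Unlit, two-sided translucent material for Niagara mesh particle splines in the Toyfox transformation intro effect.
# SplineThicken sets strand width from a Width parameter multiplied by object scale.
# Particle color times a tint and high Emissive value feeds an emissive adjustment function.
# Cell noise and dissolve textures, panned with per-particle random tiling, offset and speed, are multiplied, sharpened by a power and scaled by particle alpha.
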Material /ToyFoxGameplay/Assets/FX/TransformationIntroFX/M_Toyfox_Transform_Spline_D
graph {
 properties {
  bAutomaticallySetUsageInEditor=false
  bUsedWithNiagaraMeshParticles=true
  blendMode=BLEND_Translucent
  shadingModel=MSM_Unlit
  twoSided=true
  usedShadingModels=Unlit
 }
 nodes [
  n1 = ParticleRandom()
  n2 = ScalarParameter() {defaultValue=3, parameterName=Param_1}
  n3 = Multiply(A=@n2) {constB=2}
  n4 = Multiply(A=Add(A=TextureCoordinate(), B=@n1), B=AppendVector(A=ScalarParameter() {defaultValue=0.1}, B=LinearInterpolate(A=@n2, B=@n3, Alpha=@n1)))
  n5 = Constant()
  n6 = Panner(Coordinate=@n4, Speed=AppendVector(A=@n5, B=LinearInterpolate(Alpha=@n1) {constA=1.5, constB=3})) {speedY=2.2}
  Divide(A=@n6)
  n7 = ParticleColor()
  n8 = VectorParameter() {parameterName=Color, defaultValue=<rgba 1 0.3939 0.0494 1>}
  n9 = ScalarParameter() {defaultValue=1000, parameterName=Emissive}
  MaterialFunctionCall(Emissive=Multiply(A=@n7.RGB, B=Multiply(A=@n8.RGB, B=@n9))) {materialFunction=&MF_FX_EmissiveAdjustments}
  n10 = Multiply(
   A=ScalarParameter() {defaultValue=0.1, parameterName=Width}
   B=MaterialFunctionCall() {materialFunction=&ObjectScale}.Scale XYZ
  )
  MaterialFunctionCall(WidthBase=@n10, WidthTip=@n10) {materialFunction=&SplineThicken}
  n11 = LinearInterpolate(Alpha=@n1) {constA=1.2, constB=2.5}
  n12 = TextureSample(UVs=Panner(Coordinate=@n4, Speed=AppendVector(A=@n5, B=@n11)) {speedY=1.5}) {samplerType=SAMPLERTYPE_Grayscale, texture=/Game/Weapons/FORT_Melee/Pickaxe_Cube_Ninja_Male/FX/T_CellNoise}
  n13 = TextureSample(UVs=@n6) {texture=/Game/CharClasses/Outlander/Gadgets/AntiMaterialCharge/Materials/T_AntiMaterialCharge_Dissolve_01}
  Multiply(A=@n7.A, B=Power(Base=Multiply(A=@n12.R, B=@n13.R), Exp=ScalarParameter() {defaultValue=2, parameterName=Param_2}) {constExponent=1.5})
 ]
}
