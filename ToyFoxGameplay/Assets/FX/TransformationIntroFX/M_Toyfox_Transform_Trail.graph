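# Translucent unlit Niagara ribbon material for the Toyfox transformation intro trail.
# A Color parameter multiplied by particle color feeds the MF_FX_EmissiveAdjustments function.
# A mask multiplies particle alpha by a powered sine of the V gradient, shaping the ribbon across its width.
# The U gradient on the second UV channel adds a sharp ramp at the start and a linear fade toward the end.
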Material /ToyFoxGameplay/Assets/FX/TransformationIntroFX/M_Toyfox_Transform_Trail
graph {
 properties {
  bAutomaticallySetUsageInEditor=false
  bUsedWithNiagaraRibbons=true
  blendMode=BLEND_Translucent
  shadingModel=MSM_Unlit
  usedShadingModels=Unlit
 }
 nodes [
  n1 = ParticleColor()
  MaterialFunctionCall(Emissive=Multiply(A=VectorParameter() {parameterName=Color, defaultValue=<rgba 1 1 1 1>}.RGB, B=@n1.RGB)) {materialFunction=&MF_FX_EmissiveAdjustments}
  n2 = MaterialFunctionCall() {materialFunction=&LinearGradient}
  n3 = Saturate(None=Sine(None=Divide(A=@n2.VGradient)))
  n4 = MaterialFunctionCall(UV Channel=TextureCoordinate() {coordinateIndex=1}) {materialFunction=&LinearGradient}
  n5 = Multiply(A=@n4.UGradient) {constB=20}
  Multiply(A=Multiply(A=Power(Base=@n3), B=@n1.A), B=Multiply(A=OneMinus(None=@n4.UGradient), B=Saturate(None=@n5)))
 ]
}
