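# Niagara air-burst effect for the ToyFox transformation combine intro, attached per bone or socket to a skeletal mesh.
# A platform-dependent system count sets spawn and partition numbers, with the partition index picking each bone, color or cone orientation.
# Rays layers sprite glows and flares, some scaled by camera distance; Trail draws colored ribbons; WindCone002 renders animated cone meshes.
# User.Fade scales alpha across all three emitters.
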NiagaraSystem /ToyFoxGameplay/Assets/FX/TransformationIntroFX/NS_ToyFox_Transform_Intro_Combine_AirBurst
dependencies {
 usedDataInterfaces [
  &NiagaraDataInterfaceArrayFloat3
  &NiagaraDataInterfaceCurve
  &NiagaraDataInterfaceArrayColor
  &NiagaraDataInterfaceColorCurve
  &NiagaraDataInterfaceCamera
  &NiagaraDataInterfaceSkeletalMesh
  &NiagaraDataInterfacePlatformSet
 ]
 usedDynamicInputs [&MakeCustomIntFromBool, &IsPlatformSetActive]
 usedModules [
  &self:SystemSpawnScript.SystemScriptSource.SystemScriptGraph.NiagaraNodeAssignment_0.SetVariables_#1
  &self:SystemSpawnScript.SystemScriptSource.SystemScriptGraph.NiagaraNodeAssignment_1.SetVariables_#2
  &SystemState
 ]
}
emitters {
 Rays {
  emitterUpdateScript [
   &EmitterState(Life Cycle Mode=::NewEnumerator1, Loop Behavior=::NewEnumerator1, Loop Duration Mode=::NewEnumerator1, UseLoopDelay=true)
   &SpawnBurst_Instantaneous(Spawn Count=$System.IsLowLimitToFive, Use Spawn Probability=false, UseLoopCountLimit=false)
  ]
  particleSpawnScript [
   &PartitionParticles(Number Of Partitions=$System.IsLowLimitToFive)
   &InitializeParticle(
    Color=&SelectLinearColorFromArray(
     Array Sampling Mode=::NewEnumerator1
     Direct Array Index=$Particles.PartitionParticles.Partition
     Color Selection Array=DataInterface() {propertyValues {ColorData [
        <rgba 0 0.201 1 1>
        <rgba 0.5174 0.2574 0.0027 1>
        <rgba 0.1649 0.1649 0.1649 1>
        <rgba 1 0 0 1>
        <rgba 1 0.0058 0.5699 1>
        <rgba 0.5347 0.2835 0.0415 1>
       ]}}
     Random Seed=?
    )
    Lifetime=0.6
    Mesh Scale=<0.1 0.1 0.05>
    Mesh Scale Mode=::NewEnumerator4
    Sprite Size Mode=::NewEnumerator3
    Uniform Sprite Size=5000
    UsePositionOffset=false
    Write Material Random=true
   )
   /SetVariables(Particles.Dy:Vector4f=<0.5 1 1 1>) {isSetParameters=true}
   /SetVariables(Particles.DySpriteGlow:Vector4f=<0.4 1 0 0>) {isSetParameters=true}
   /SetVariables(Particles.PDCameraOffset=300) {isSetParameters=true}
  ]
  particleUpdateScript [
   &ParticleState(Kill Particles When Lifetime Has Elapsed=false, Let Infinitely Lived Particles Die When Emitter Deactivates=true)
   &SkeletalMeshLocation(
    Bone / Socket Index=$Particles.PartitionParticles.Partition
    Bone Sampling Mode=::NewEnumerator4
    MaskSpawnGroup=false
    Skeletal Mesh=$System.SK_Mesh
   )
   &ScaleColor(Scale Alpha=$User.Fade, Scale RGB=&VectorFromFloat(Value=10), ScaleA=true, ScaleRGB=true)
   /SetVariables(
    Particles.PartDistanceFar:NiagaraFloat=&RemapRange(
     Input Value To Remap=&DistanceBetweenPositions(
      A=&ReturnCameraPosition(Coordinate Space:ENiagaraCoordinateSpace=::Simulation, Camera Query:NiagaraDataInterfaceCamera=DataInterface())
      B=$Particles.SkeletalMeshLocation.SampledPosition
     )
     Input Min=10000
     Input Max=80000
     Output Min=0
     Output Max=1
    )
    Particles.PartDistanceNear:NiagaraFloat=&RemapRange(
     Input Value To Remap=&DistanceBetweenPositions(
      A=&ReturnCameraPosition(Coordinate Space:ENiagaraCoordinateSpace=::Simulation, Camera Query:NiagaraDataInterfaceCamera=DataInterface())
      B=$Particles.SkeletalMeshLocation.SampledPosition
     )
     Input Min=10000
     Input Max=30000
     Output Min=0
     Output Max=1
    )
   ) {isSetParameters=true}
   /SetVariables(
    Particles.SoftGlowColor:LinearColor=&Multiply_LinearColorByFloat(Linear Color=$Particles.Color, Float=0.2)
    Particles.SoftGlowSize:Vector3f=&Multiply_VectorByFloat(Vector=$Particles.Scale, Float=3)
   ) {isSetParameters=true}
   /SetVariables(
    Particles.CrossFlareColor:LinearColor=&Multiply_LinearColorByFloat(Linear Color=$Particles.Color, Float=0.3)
    Particles.CrossFlareSize:Vector2f=&Multiply_Vector2D(
     A=$Particles.SpriteSize
     B=&Multiply_Vector2D_ByFloat(
      Float=&Lerp_Float(A=0.2, B=1, Alpha=$Particles.PartDistanceNear)
      Vector2D=&Vector2DFromFloat(Value=&RandomRangeFloat(Minimum=10, Maximum=15, Evaluation Type=::NewEnumerator1))
     )
    )
   ) {isSetParameters=true}
   /SetVariables(
    Particles.HugeFlareColor:LinearColor=&Multiply_LinearColorByFloat(Linear Color=$Particles.Color, Float=0.4)
    Particles.HugeFlareColorB:LinearColor=&Multiply_LinearColorByFloat(
     Linear Color=&HueShiftLinearColor(Hue=0.15, Saturation=0.8, Value=1, Linear Color=$Particles.Color)
     Float=&RandomRangeFloat(Minimum=0.3, Maximum=0.5, Evaluation Type=::NewEnumerator1)
    )
    Particles.HugeFlareColorC:LinearColor=&Multiply_LinearColorByFloat(
     Linear Color=<rgba 3 1.4234 0.0384 1>
     Float=&RandomRangeFloat(Minimum=0.6, Maximum=1, Evaluation Type=::NewEnumerator1)
    )
    Particles.HugeFlareRotationC:NiagaraFloat=&RandomRangeFloat(Minimum=0, Maximum=360)
    Particles.HugeFlareRotationD:NiagaraFloat=&RandomRangeFloat(Minimum=0, Maximum=360)
    Particles.HugeFlareSize:Vector2f=&Multiply_Vector2D(
     A=$Particles.SpriteSize
     B=&Multiply_Vector2D_ByFloat(Float=$Particles.PartDistanceFar, Vector2D=&MakeVector2D(X=15, Y=600))
    )
    Particles.HugeFlareSizeB:Vector2f=&Multiply_Vector2D(
     A=$Particles.SpriteSize
     B=&Multiply_Vector2D_ByFloat(Float=$Particles.PartDistanceFar, Vector2D=&MakeVector2D(X=100, Y=100))
    )
    Particles.HugeFlareSizeC:Vector2f=&Multiply_Vector2D(
     A=$Particles.SpriteSize
     B=&Multiply_Vector2D_ByFloat(Float=$Particles.PartDistanceFar, Vector2D=&MakeVector2D(X=100, Y=30))
    )
   ) {isSetParameters=true}
   /SetVariables(Particles.PDSize:Vector2f=&Multiply_Vector2D(A=$Particles.SpriteSize, B=&Vector2DFromFloat(Value=0.8))) {isSetParameters=true}
   &DynamicMaterialParameters(
    Index 0 Param 1=&FloatFromCurve(FloatCurve=DataInterface() {propertyValues {Curve {keys [
         {interpMode=RCIM_Cubic, tangentMode=RCTM_User, time=0, value=1}
         {interpMode=RCIM_Cubic, tangentMode=RCTM_User, time=0.2, value=0.5}
         {interpMode=RCIM_Cubic, tangentMode=RCTM_User, time=1, value=1}
        ]}}})
    Module.Index0Float0Write=true
    Module.Index0Float1Write=true
    Module.Index0Float2Write=true
    Module.Index0Float3Write=true
   )
  ]
  rendererClasses [&NiagaraSpriteRendererProperties]
  renderers [
   {rendererClass=&NiagaraSpriteRendererProperties}
   {rendererClass=&NiagaraSpriteRendererProperties, rendererIndex=1}
   {rendererClass=&NiagaraSpriteRendererProperties, rendererIndex=2}
   {rendererClass=&NiagaraSpriteRendererProperties, rendererIndex=3}
   {rendererClass=&NiagaraSpriteRendererProperties, rendererIndex=4}
   {rendererClass=&NiagaraSpriteRendererProperties, rendererIndex=5}
   {rendererClass=&NiagaraSpriteRendererProperties, rendererIndex=6}
   {rendererClass=&NiagaraSpriteRendererProperties, rendererIndex=7}
  ]
 }
 Trail {
  emitterUpdateScript [
   &EmitterState(Life Cycle Mode=::NewEnumerator1, Loop Behavior=::NewEnumerator1, Loop Duration=999, UseLoopDelay=false)
   &SpawnRate(SpawnRate=100, Use Spawn Probability=false)
  ]
  particleSpawnScript [
   &PartitionParticles(Number Of Partitions=$System.IsLowLimitToFive)
   &InitializeParticle(
    Lifetime=6
    Ribbon Width=2500
    Ribbon Width Mode=::NewEnumerator1
    Sprite Size Mode=::NewEnumerator3
    Uniform Sprite Size=5000
    UsePositionOffset=false
    Write Material Random=true
   )
   &SkeletalMeshLocation(
    Bone / Socket Index=$Particles.PartitionParticles.Partition
    Bone Sampling Mode=::NewEnumerator4
    MaskSpawnGroup=false
    Skeletal Mesh=$System.SK_Mesh
   )
  ]
  particleUpdateScript [
   &ParticleState()
   &ScaleColor(Scale Alpha=$User.Fade, ScaleA=true, ScaleRGB=true)
   &ScaleColor(
    Scale Alpha=&Multiply_Float(
     A=&Lerp_Float(
      A=0
      B=0.05
      Alpha=&FloatFromCurve(FloatCurve=DataInterface() {propertyValues {Curve {
          defaultValue=1
          keys [{interpMode=RCIM_Cubic, tangentMode=RCTM_User, time=0, value=1}, {interpMode=RCIM_Cubic, tangentMode=RCTM_User, time=1, value=0}]
         }}})
     )
     B=&FloatFromCurve(
      FloatCurve=DataInterface() {propertyValues {Curve {
         defaultValue=0
         keys [{interpMode=RCIM_Cubic, tangentMode=RCTM_User, time=0, value=0}, {interpMode=RCIM_Cubic, tangentMode=RCTM_User, time=3, value=1}]
        }}}
      CurveIndex=&Add_Float(A=$System.Age, B=20)
     )
    )
    ScaleA=true
    ScaleRGB=true
   ) {module=ScaleColor001}
   /SetVariables(Particles.ColorRibbon:LinearColor=&MultiplyLinearColor(
     Linear Color=&SelectLinearColorFromArray(
      Array Sampling Mode=::NewEnumerator1
      Direct Array Index=$Particles.PartitionParticles.Partition
      Color Selection Array=DataInterface() {propertyValues {ColorData [
         <rgba 0 0.201 1 1>
         <rgba 1 0.4975 0.0052 1>
         <rgba 0.1649 0.1649 0.1649 1>
         <rgba 1 0 0 1>
         <rgba 1 0.0058 0.5699 1>
         <rgba 0.2083 0.2083 0.2083 1>
        ]}}
      Random Seed=?
     )
     Scale Factor=&MakeVector4_FromLinearColor(LinearColor=&MakeLinearColorFromVectorAndFloat(
       "Vector (RGB)"=&VectorFromFloat(Value=5)
       "Float (Alpha)"=&Multiply_Float(
        A=$User.Fade
        B=&Lerp_Float(
         A=0
         B=0.3
         Alpha=&FloatFromCurve(FloatCurve=DataInterface() {propertyValues {Curve {
             defaultValue=1
             keys [{interpMode=RCIM_Cubic, tangentMode=RCTM_User, time=0, value=1}, {interpMode=RCIM_Cubic, tangentMode=RCTM_User, time=1, value=0}]
            }}})
        )
       )
      ))
    )) {isSetParameters=true}
  ]
  rendererClasses [&NiagaraRibbonRendererProperties]
  renderers [
   {rendererClass=&NiagaraRibbonRendererProperties}
   {rendererClass=&NiagaraRibbonRendererProperties, rendererIndex=1}
  ]
 }
 WindCone002 {
  emitterUpdateScript [
   &EmitterState(
    Inactive Response=::NewEnumerator1
    Life Cycle Mode=::NewEnumerator1
    Loop Behavior=::NewEnumerator1
    Loop Duration Mode=::NewEnumerator1
    UseLoopDelay=false
   )
   &SpawnBurst_Instantaneous(Spawn Count=$System.IsLowLimitToFive, Use Spawn Probability=false, UseLoopCountLimit=false)
  ]
  particleSpawnScript [
   &PartitionParticles(Number Of Partitions=$System.IsLowLimitToFive)
   &InitializeParticle(
    Color=<rgba 1 1 1 0.15>
    Lifetime=&RandomRangeFloat(Minimum=0.6, Maximum=0.4)
    Mesh Scale=&Multiply_VectorByFloat(
     Vector=&SelectVectorFromArray(
      Direct Array Index=$Particles.PartitionParticles.Partition
      Vector Selection Array=DataInterface() {propertyValues {FloatData [<1 1 1>, <1 2 2>, <1 1 1>, <1.5 1.5 2>, <1 1 1>, <1 1 1>]}}
     )
     Float=0.5
    )
    Mesh Scale Mode=::NewEnumerator4
    UsePositionOffset=true
    Write Material Random=true
   )
   &InitialMeshOrientation(
    Mesh Axis To Orient=&SelectVectorFromArray(
     Direct Array Index=$Particles.PartitionParticles.Partition
     Vector Selection Array=DataInterface() {propertyValues {FloatData [<-1 0 0>, <1 0.3 0>, <0 -1 0>, <0.1 1 0>, <-1 2 0>, <-0.4 -1 0>]}}
    )
    Mesh Orientation Mode=::NewEnumerator2
    Rotation Coordinate Space=::NewEnumerator0
    Use Rotation Vector=true
   )
  ]
  particleUpdateScript [
   &ParticleState(Kill Particles When Lifetime Has Elapsed=false, Let Infinitely Lived Particles Die When Emitter Deactivates=true)
   &SkeletalMeshLocation(
    Bone / Socket Index=$Particles.PartitionParticles.Partition
    Bone Sampling Mode=::NewEnumerator4
    MaskSpawnGroup=false
    Orientation Sampling=::NewEnumerator1
    Skeletal Mesh=$System.SK_Mesh
   )
   &DynamicMaterialParameters(
    Index 0 Param 1=&Add_Float(
     A=&RandomRangeFloat(Minimum=0, Maximum=999)
     B=&Multiply_Float(A=$Engine.Time, B=6)
    )
    Index 0 Param 2=10
    Module.Index0Param0Write=true
    Module.Index0Param1Write=true
    Module.Index0Param2Write=true
    Module.Index0Param3Write=true
   )
   &ScaleColor(Scale Alpha=$User.Fade, ScaleA=true, ScaleRGB=true) {module=ScaleColor001}
  ]
  rendererClasses [&NiagaraMeshRendererProperties]
  renderers [
   {rendererClass=&NiagaraMeshRendererProperties, rendererIndex=0}
   {rendererClass=&NiagaraMeshRendererProperties, rendererIndex=1}
   {rendererClass=&NiagaraMeshRendererProperties, rendererIndex=2}
  ]
 }
}
summary {
 systemName=NS_ToyFox_Transform_Intro_Combine_AirBurst
 emitters [
  {emitterName=WindCone002, rendererClasses [&NiagaraMeshRendererProperties]}
  {emitterName=Trail, rendererClasses [&NiagaraRibbonRendererProperties]}
  {emitterName=Rays, rendererClasses [&NiagaraSpriteRendererProperties]}
 ]
 userVariables [User.Fade {type=NiagaraFloat, defaultValue=1, description=""}]
}
